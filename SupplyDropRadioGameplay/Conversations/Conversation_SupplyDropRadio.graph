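# Conversation database for the Supply Drop Radio item, defining the choice menu entered through its item entry tag.
# The entry point leads to a speech node that fans out into five parallel SupplyDropRadio task branches, each ending the graph.
# Each task reuses the same task, requirement and choice classes, differing by conversation index.
# Only the default-index task shows explicit hover and confirm sounds.
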ConversationDatabase /SupplyDropRadioGameplay/Conversations/Conversation_SupplyDropRadio
entry_points [{tag=Conversation.Entry.Item.SupplyDropRadio, destinations [#1]}]
nodes [
 ConversationEntryPointNode_1 {
  id=#1
  class=ConversationEntryPointNode
  outputs [#2]
  properties {entryTag {tagName=Conversation.Entry.Item.SupplyDropRadio}}
 }
 Speech_C_3 {
  id=#2
  class=Speech_C
  outputs [
   #3
   #4
   #5
   #6
   #7
  ]
  properties {voiceOverTag {tagName=None}}
 }
 ConvoTask_SupplyDropRadio_C_6 {
  id=#7
  class=ConvoTask_SupplyDropRadio_C
  outputs []
  properties {conversationIndex=4}
  sub_nodes [
   ConvoRequirement_SupplyDropRadio_C_9 {class=ConvoRequirement_SupplyDropRadio_C}
   ConvoChoice_SupplyDropRadio_C_5 {class=ConvoChoice_SupplyDropRadio_C}
  ]
 }
 ConvoTask_SupplyDropRadio_C_7 {
  id=#6
  class=ConvoTask_SupplyDropRadio_C
  outputs []
  properties {conversationIndex=3}
  sub_nodes [
   ConvoRequirement_SupplyDropRadio_C_10 {class=ConvoRequirement_SupplyDropRadio_C}
   ConvoChoice_SupplyDropRadio_C_6 {class=ConvoChoice_SupplyDropRadio_C}
  ]
 }
 ConvoTask_SupplyDropRadio_C_8 {
  id=#5
  class=ConvoTask_SupplyDropRadio_C
  outputs []
  properties {conversationIndex=2}
  sub_nodes [
   ConvoRequirement_SupplyDropRadio_C_11 {class=ConvoRequirement_SupplyDropRadio_C}
   ConvoChoice_SupplyDropRadio_C_7 {class=ConvoChoice_SupplyDropRadio_C}
  ]
 }
 ConvoTask_SupplyDropRadio_C_9 {
  id=#4
  class=ConvoTask_SupplyDropRadio_C
  outputs []
  properties {conversationIndex=1}
  sub_nodes [
   ConvoRequirement_SupplyDropRadio_C_12 {class=ConvoRequirement_SupplyDropRadio_C}
   ConvoChoice_SupplyDropRadio_C_8 {class=ConvoChoice_SupplyDropRadio_C}
  ]
 }
 ConvoTask_SupplyDropRadio_C_1 {
  id=#3
  class=ConvoTask_SupplyDropRadio_C
  outputs []
  properties {
   choiceSound=/Game/Athena/Sounds/NPCs/Conversation/Conversation_Confirm_Cue
   hoverChoiceSound=/Game/Athena/Sounds/NPCs/Conversation/Conversation_Hover_Cue
  }
  sub_nodes [
   ConvoChoice_SupplyDropRadio_C_2 {class=ConvoChoice_SupplyDropRadio_C}
   ConvoRequirement_SupplyDropRadio_C_1 {class=ConvoRequirement_SupplyDropRadio_C}
  ]
 }
]
speakers []
