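; Conversation choice for the Supply Drop Radio that builds the client option entry a player sees.
; FillChoice takes the player controller from the conversation context and casts the task being considered to ConvoTask_SupplyDropRadio to fetch that player's drop data.
; When the data is valid, the entry uses its name and gameplay tag and is marked available or unavailable by the task's stock check.
; Otherwise it returns the parent's default entry.
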(fn FillChoice
  (params (Context :type Conversation%20Context%20Structure%20%28by%20ref%29))
  ;[0.02187 0.08514 0.1354] Build out the client option entry based on the stock availability of the option.
  (Utilities|FlowControl|Sequence
    (:then_0
      (bind _returnvalue (Class|FortConversationContextLibrary|GetParticipantPlayerController Context))
      (bind _returnvalue_2 (Class|FortConversationContextLibrary|GetTaskBeingConsidered Context))
      (bind _asconvo_task_supply_drop_radio (Utilities|Casting|CastToConvoTask_SupplyDropRadio _returnvalue_2))
      (bind _output_get (LocalVariables|SetConvoTask _asconvo_task_supply_drop_radio))
      (bind (_bvalid _data) (Class|ConvoTaskSupplyDropRadio|GetSupplyDropRadioData _output_get _returnvalue))
      (if _bvalid
        (return (Utilities|Struct|MakeClientConversationOptionEntry (.Name_15_299F1575424B7AC4048B16BE4575858A (Utilities|Struct|BreakSupplyDropRadioDataStruct _data)) (GameplayTags|MakeGameplayTagContainerfromTag (.GameplayTag_5_542794F64B7B42F5AFFB4F99DE156B3F (Utilities|Struct|BreakSupplyDropRadioDataStruct _data))) (select (Class|ConvoTaskSupplyDropRadio|IsStockAvailable (LocalVariables|GetConvoTask) _returnvalue) "UserChoiceAvailable" "UserChoiceUnavailable")))))
    (:then_1
      (bind _choiceentry (|CallParentFunction Context))
      (return _choiceentry))))
