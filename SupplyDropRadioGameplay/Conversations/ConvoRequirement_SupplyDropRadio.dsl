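; Conversation requirement that decides whether a supply drop radio choice is available, visible but blocked, or hidden for the participant player.
; It casts the task being considered to ConvoTask_SupplyDropRadio and asks it whether it is enabled and has stock for that player's controller.
; Enabled with stock passes, and enabled without stock fails but stays visible.
; A Sequence node's second branch serves as the fallback, failing and hiding the choice in other cases.
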(fn IsRequirementSatisfied
  (params (Context :type Conversation%20Context%20Structure%20%28by%20ref%29))
  ;[0 0.07812 0.06788] Returns if there's enough stock available in the item.
  (Utilities|FlowControl|Sequence
    (:then_0
      (bind _returnvalue (Class|FortConversationContextLibrary|GetParticipantPlayerController Context))
      (bind _returnvalue_2 (Class|FortConversationContextLibrary|GetTaskBeingConsidered Context))
      (bind _asconvo_task_supply_drop_radio (Utilities|Casting|CastToConvoTask_SupplyDropRadio _returnvalue_2))
      (if (Class|ConvoTaskSupplyDropRadio|IsEnabled _asconvo_task_supply_drop_radio _returnvalue)
        (if (Class|ConvoTaskSupplyDropRadio|IsStockAvailable _asconvo_task_supply_drop_radio _returnvalue)
          (return "Passed")
          (else
            (return "FailedButVisible")))))
    (:then_1
      (return "FailedAndHidden"))))
