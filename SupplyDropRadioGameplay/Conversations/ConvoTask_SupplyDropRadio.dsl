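; Conversation task node for one indexed option in the Supply Drop Radio conversation, used to call in a supply drop.
; On execution and in client effects it sends an option-select gameplay event to the player pawn, carrying the conversation index as magnitude.
; Data, enabled state and stock queries look up that index on BP_SupplyDropRadioGameStateComponent.
; The choice display asset comes from that data, falling back to the parent implementation.
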(fn ExecuteClientEffects
  (params (Context :type Conversation%20Context%20Structure%20%28by%20ref%29))
  ;[0 0.1719 0.1608] ExecuteClientEffects: Sends a gameplay event to the player pawn with the selected option index in the payload.
  (CallFunction|SendOptionSelectGameplayEventtoPlayerPawn Context)
  (return))

(fn ExecuteTaskNode
  (params (Context :type Conversation%20Context%20Structure%20%28by%20ref%29))
  (bind _context (Variables|GetContext))
  ;[0 0.06329 0.06771] ExecuteTaskNote: Call in the supply drop.
  (Utilities|FlowControl|Sequence
    (:then_0
      (bind _returnvalue (Class|FortConversationContextLibrary|GetParticipantActorNPC _context))
      (Utilities|Casting|CastToB_SupplyDropRadio_Handheld_Item _returnvalue)
      (CallFunction|SendOptionSelectGameplayEventtoPlayerPawn _context))
    (:then_1
      (bind _returnvalue_2 (Conversation|AdvanceConversation Context))
      (return _returnvalue_2))))

(fn GetConversationIndex
  ;[0.03814 0.05011 0.1198] Return the conversation index.
  (return (Variables|Default|GetConversationIndex)))

(fn GetDisplayAssetForChoice
  (params (Context :type Conversation%20Context%20Structure%20%28by%20ref%29) (InOptionEntry :type Client%20Conversation%20Option%20Entry%20Structure%20%28by%20ref%29))
  ;[0.09286 0.0344 0.1094] Returns the display asset to use for the choice.
  (Utilities|FlowControl|Sequence
    (:then_0
      (bind _returnvalue (Class|FortConversationContextLibrary|GetParticipantActorPlayer Context))
      (bind (_bvalid _data) (CallFunction|GetSupplyDropRadioData _returnvalue))
      (if _bvalid
        (if (Utilities|IsValidSoftObjectReference (.ChoiceImageSoftObj_29_E219BD914583592495E1F38287CB7341 (Utilities|Struct|BreakSupplyDropRadioDataStruct _data)))
          (return (.ChoiceImageSoftObj_29_E219BD914583592495E1F38287CB7341 (Utilities|Struct|BreakSupplyDropRadioDataStruct _data))))))
    (:then_1
      (bind _returnvalue_2 (|CallParentFunction Context InOptionEntry))
      (return _returnvalue_2))))

(fn GetSupplyDropRadioData
  (params (WorldContextObject :type Object%20Reference))
  ;[0.2396 0.01971 0.116] Returns the supply drop radio data from the game state component based on the conversation index.
  (Utilities|FlowControl|Sequence
    (:then_0
      (bind _component (CallFunction|GetSupplyDropRadioGameStateComponent WorldContextObject))
      (Utilities|IsValid _component
        (:Is%20Valid
          (bind (_isvalid _data) (Class|BPSupplyDropRadioGameStateComponent|GetSupplyDropRadioDataUsingConversationIndex _component (Variables|Default|GetConversationIndex)))
          (return _isvalid _data))))
    (:then_1
      (return false 0))))

(fn GetSupplyDropRadioGameStateComponent
  (params (WorldContextObject :type Object%20Reference))
  ;[0.01993 0.1283 0.1875] Returns the supply drop radio game state component.
  (Utilities|FlowControl|Sequence
    (:then_0
      (bind _output_get (LocalVariables|SetGameStateAthena (Fortnite|Athena|GameState|GetGameStateAthenafromContext WorldContextObject)))
      (Utilities|IsValid _output_get
        (:Is%20Valid
          (return (Actor|GetComponentbyClass _output_get "/SupplyDropRadioGameplay/Gameplay/BP_SupplyDropRadioGameStateComponent.BP_SupplyDropRadioGameStateComponent_C")))))
    (:then_1
      (return 0))))

(fn IsEnabled
  (params (WorldContextObject :type Object%20Reference))
  ;[0.1771 0.02411 0.1677] Returns if a particular option is enabled.
  (Utilities|FlowControl|Sequence
    (:then_0
      (bind _component (CallFunction|GetSupplyDropRadioGameStateComponent WorldContextObject))
      (Utilities|IsValid _component
        (:Is%20Valid
          (bind _benabled (Class|BPSupplyDropRadioGameStateComponent|IsConversationIndexEnabled _component (Variables|Default|GetConversationIndex)))
          (return _benabled))))
    (:then_1
      (return false))))

(fn IsStockAvailable
  (params (WorldContextObject :type Object%20Reference))
  ;[0.5156 0.4009 0.07125] Returns if there's stock available for the specified conversation index.
  (Utilities|FlowControl|Sequence
    (:then_0
      (bind _component (CallFunction|GetSupplyDropRadioGameStateComponent WorldContextObject))
      (Utilities|IsValid _component
        (:Is%20Valid
          (return (Class|BPSupplyDropRadioGameStateComponent|IsStockAvailableforConversationIndex _component (Variables|Default|GetConversationIndex)) (.Count (Class|BPSupplyDropRadioGameStateComponent|IsStockAvailableforConversationIndex _component (Variables|Default|GetConversationIndex)))))))
    (:then_1
      (return false 0))))

(fn SendOptionSelectGameplayEventToPlayerPawn
  (params (Context :type Conversation%20Context%20Structure%20%28by%20ref%29))
  ;[0 0.2031 0.1732] Sends the option selected as a gameplay event to the player pawn.
  (bind _returnvalue (Class|FortConversationContextLibrary|GetParticipantPlayerController Context))
  (Utilities|IsValid (Game|Player|GetPlayerPawn _returnvalue)
    (:Is%20Valid
      (Ability|SendGameplayEventtoActor (Game|Player|GetPlayerPawn _returnvalue) "(TagName=\"Event.SupplyDropRadio.Conversation.Option.Select\")" (Utilities|Struct|MakeGameplayEventData "(TagName=\"Event.SupplyDropRadio.Conversation.Option.Select\")" 0 0 0 0 "()" "(GameplayTags=)" "(GameplayTags=)" (Math|Conversions|ToFloat(Integer) (Variables|Default|GetConversationIndex)))))))
